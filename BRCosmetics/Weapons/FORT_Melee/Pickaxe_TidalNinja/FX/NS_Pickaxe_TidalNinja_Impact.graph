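# Niagara impact effect for the Tidal Ninja pickaxe, built from short-lived burst emitters.
# Layers glow sprites, a fresnel sphere, a flare mesh, two spinning expanding rings, a shrinking colored sphere and curl-noise sparks.
# User.VariantSwitch picks each variant's look through color arrays, blended color curves and a VisibilityTag on the multi-renderer sphere.
# Front-end modifiers and bIsFrontEndPreview adjust scale, camera offset, alpha and spark speed for menu previews.
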NiagaraSystem /BRCosmetics/Weapons/FORT_Melee/Pickaxe_TidalNinja/FX/NS_Pickaxe_TidalNinja_Impact
dependencies {
 usedDataInterfaces [
  &NiagaraDataInterfaceArrayColor
  &NiagaraDataInterfaceColorCurve
  &NiagaraDataInterfaceCurve
  &NiagaraDataInterfaceArrayInt32
  &NiagaraDataInterfaceArrayFloat
 ]
 usedDynamicInputs []
 usedModules [
  &self:SystemSpawnScript.SystemScriptSource.SystemScriptGraph.NiagaraNodeAssignment_1.SetVariables_#1
  &SystemState
  &self:SystemSpawnScript.SystemScriptSource.SystemScriptGraph.NiagaraNodeAssignment_0.SetVariables_#2
 ]
}
emitters {
 Dark_glow {
  emitterUpdateScript [
   &EmitterState()
   &SpawnBurst_Instantaneous(Use Spawn Probability=false, UseLoopCountLimit=false)
  ]
  particleSpawnScript [
   &InitializeParticle(Lifetime=0.35, Sprite Size Mode=::NewEnumerator3, Uniform Sprite Size=150, UsePositionOffset=false, Write Material Random=true)
   &CameraOffset(Camera Offset Amount=&FN_FrontEndFloatModifier(Gameplay=30, Front End=0))
  ]
  particleUpdateScript [
   &ParticleState()
   &ScaleSpriteSize(
    Uniform Curve Scale=&Multiply_Float(A=1.5, B=$System.FX size scale)
    Uniform Curve Sprite Scale=DataInterface() {propertyValues {Curve {keys [{tangentMode=RCTM_User, time=0, value=1}, {tangentMode=RCTM_User, time=1, value=0}]}}}
   )
   &ScaleColor(
    Scale Alpha=&FloatFromCurve(
     FloatCurve=DataInterface() {propertyValues {Curve {keys [
         {tangentMode=RCTM_User, time=0, value=0}
         {tangentMode=RCTM_User, time=0.2, value=1}
         {arriveTangent=-4.9829, interpMode=RCIM_Cubic, leaveTangent=-4.9829, tangentMode=RCTM_User, time=0.6, value=1.025}
         <at 1 0 RCIM_Cubic>
        ]}}}
     Scale Curve=&FN_FrontEndFloatModifier(Gameplay=1.5, Front End=0)
    )
    Scale RGB=&MakeVectorFromLinearColorRGB(Color=&SelectLinearColorFromArray(
      Array Sampling Mode=::NewEnumerator1
      Direct Array Index=&MakeIntFromFloat(Float=$User.VariantSwitch)
      Color Selection Array=DataInterface() {propertyValues {ColorData [<rgba 0.0549 0 0.1 1>, <rgba 0.0215 0.0054 0.1 1>, <rgba 0.0059 0.0315 0.1 1>, <rgba 0.1 0.0335 0.005 1>, <rgba 0.0143 0.0758 0.1 1>]}}
      Random Seed=?
     ))
    ScaleA=true
    ScaleRGB=true
   )
   &DynamicMaterialParameters(
    Index 0 Param 1=2
    Index 0 Param 2=1
    Index 2 Param 1=&RandomRangeFloat(Minimum=0.4, Maximum=0.6)
    Index 2 Param 2=1
    Index 2 Param 3=1
    Index 2 Param 4=1
    Module.Index0Param0Write=true
    Module.Index0Param1Write=true
    Module.Index0Param2Write=true
    Module.Index0Param3Write=true
    Module.Index2Param0Write=true
    Module.Index2Param1Write=true
    Module.Index2Param2Write=true
    Module.Index2Param3Write=true
    Write Parameter Index 2=true
   )
   &FN_Sprite_Attributes(
    Color=&MakeLinearColorFromVectorAndFloat(
     "Vector (RGB)"=&Multiply_Vector(
      A=&MakeVectorFromLinearColorRGB(Color=&SelectLinearColorFromArray(
        Array Sampling Mode=::NewEnumerator1
        Direct Array Index=&MakeIntFromFloat(Float=$User.VariantSwitch)
        Color Selection Array=DataInterface() {propertyValues {ColorData [<rgba 0.3891 0.0837 1 1>, <rgba 0.0856 0.2534 1 1>, <rgba 0.0735 0.4957 1 1>, <rgba 1 0.3737 0.0606 1>, <rgba 0.7444 0.3092 0.7917 2>]}}
        Random Seed=?
       ))
      B=<1 1 1>
     )
     "Float (Alpha)"=&Multiply_Float(
      A=&FloatFromCurve(FloatCurve=DataInterface() {propertyValues {Curve {keys [<at 0 0>,<at 0.16 1>,<at 1 1>]}}})
      B=1
     )
    )
    Create Color Attribute=true
    Create SpriteSize Attribute=true
    SpriteSize=&Multiply_Vector2D(
     A=<150 150>
     B=&Multiply_Vector2D_ByFloat(
      Float=&FloatFromCurve(
       FloatCurve=DataInterface() {propertyValues {Curve {keys [<at 0 1 RCIM_Cubic>, {arriveTangent=-1.5694, tangentMode=RCTM_User, time=0.8, value=0}]}}}
       Scale Curve=$System.FX size scale
      )
      Vector2D=<1.05 1.05>
     )
    )
    Write to Module Namespace=true
   )
   &FN_Sprite_Attributes(
    Color=&MakeLinearColorFromVectorAndFloat(
     "Vector (RGB)"=&Multiply_Vector(
      A=&MakeVectorFromLinearColorRGB(Color=&SelectLinearColorFromArray(
        Array Sampling Mode=::NewEnumerator1
        Direct Array Index=&MakeIntFromFloat(Float=$User.VariantSwitch)
        Color Selection Array=DataInterface() {propertyValues {ColorData [<rgba 3.3333 0 10 1>, <rgba 0 1.0745 10 1>, <rgba 0.1704 1.5164 2 1>, <rgba 2 0.9175 0.1443 1>, <rgba 2.5611 0.7232 3 1>]}}
        Random Seed=?
       ))
      B=<2 2 2>
     )
     "Float (Alpha)"=&Multiply_Float(
      A=&FloatFromCurve(
       FloatCurve=DataInterface() {propertyValues {Curve {keys [<at 0 0>,<at 0.14 1>,<at 0.56 1>,<at 0.7 0>]}}}
       Scale Curve=&FN_FrontEndFloatModifier(Gameplay=1, Front End=0.5)
      )
      B=1
     )
    )
    Create Color Attribute=true
    Create SpriteSize Attribute=true
    SpriteSize=&Multiply_Vector2D(
     A=<150 150>
     B=&Multiply_Vector2D_ByFloat(
      Float=&FloatFromCurve(
       FloatCurve=DataInterface() {propertyValues {Curve {keys [<at 0 1.5 RCIM_Cubic>, {arriveTangent=-1.5694, tangentMode=RCTM_User, time=0.7, value=0}]}}}
       Scale Curve=$System.FX size scale
      )
      Vector2D=<1 1>
     )
    )
    Write to Module Namespace=true
   ) {module=FN_Sprite_Attributes001}
   &FN_Sprite_Attributes(
    Color=&MakeLinearColorFromVectorAndFloat(
     "Vector (RGB)"=&Multiply_Vector(
      A=&MakeVectorFromLinearColorRGB(Color=&SelectLinearColorFromArray(
        Array Sampling Mode=::NewEnumerator1
        Direct Array Index=&MakeIntFromFloat(Float=$User.VariantSwitch)
        Color Selection Array=DataInterface() {propertyValues {ColorData [<rgba 0.7692 0.1538 2 1>, <rgba 0.1747 0.9433 2 1>, <rgba 0.9054 0 2 1>, <rgba 2 0.7417 0.1126 1>, <rgba 0.3795 1.0424 2 1>]}}
        Random Seed=?
       ))
      B=<1 1 1>
     )
     "Float (Alpha)"=&Multiply_Float(
      A=&FloatFromCurve(
       FloatCurve=DataInterface() {propertyValues {Curve {keys [<at 0 0>,<at 0.14 1>,<at 0.56 1>,<at 0.8 0>]}}}
       Scale Curve=&FN_FrontEndFloatModifier(Gameplay=1, Front End=1)
      )
      B=1
     )
    )
    Create Color Attribute=true
    Create SpriteSize Attribute=true
    SpriteSize=&Multiply_Vector2D(
     A=<140 140>
     B=&Multiply_Vector2D_ByFloat(
      Float=&FloatFromCurve(
       FloatCurve=DataInterface() {propertyValues {Curve {keys [<at 0 1 RCIM_Cubic>, {arriveTangent=-1.5694, tangentMode=RCTM_User, time=0.7, value=0}]}}}
       Scale Curve=$System.FX size scale
      )
      Vector2D=<1 1>
     )
    )
    Write to Module Namespace=true
   ) {module=FN_Sprite_Attributes002}
  ]
  rendererClasses [&NiagaraSpriteRendererProperties]
  renderers [
   {rendererClass=&NiagaraSpriteRendererProperties}
   {rendererClass=&NiagaraSpriteRendererProperties, rendererIndex=1}
   {rendererClass=&NiagaraSpriteRendererProperties, rendererIndex=2}
   {rendererClass=&NiagaraSpriteRendererProperties, rendererIndex=3}
  ]
 }
 ExplodeFresnelSphere {
  emitterUpdateScript [
   /SetVariables(Emitter.JustSphere=1) {isSetParameters=true}
   &EmitterState(Life Cycle Mode=::NewEnumerator1, Loop Behavior=::NewEnumerator1, Loop Delay=0.05, Loop Duration=1.4167, UseLoopDelay=true)
   &SpawnBurst_Instantaneous(Spawn Count=1, Use Spawn Probability=false)
   &SpawnBurst_Instantaneous(Spawn Time=1.4167, Use Spawn Probability=false, UseLoopCountLimit=false) {module=SpawnBurst_Instantaneous001}
  ]
  particleSpawnScript [
   &InitializeParticle(
    Color=&SelectLinearColorFromArray(
     Array Sampling Mode=::NewEnumerator1
     Direct Array Index=&MakeIntFromFloat(Float=$User.VariantSwitch)
     Color Selection Array=DataInterface() {propertyValues {ColorData [<rgba 0.5984 0.084 2 1>, <rgba 0.5984 0.084 2 1>, <rgba 0.1746 1.0121 2 1>, <rgba 2 1.3343 0.2038 1>, <rgba 0.4222 1.1988 2 1>]}}
     Random Seed=?
    )
    Lifetime=0.4
    Mesh Scale Mode=::NewEnumerator3
    Mesh Uniform Scale=0.5
    UsePositionOffset=false
    Write Material Random=true
   )
   &CameraOffset()
  ]
  particleUpdateScript [
   &ParticleState()
   &ScaleColor(
    Scale RGB=/Niagara/DynamicInputs/Math/LerpMultipleVectors(
     Alpha:NiagaraFloat=&Divide_Float(A=$User.VariantSwitch, B=4)
     V 1:Vector3f=&MakeVectorFromLinearColorRGB(Color=&ColorFromCurve(ColorCurve=DataInterface() {propertyValues {
         AlphaCurve {keys [<at 0 1>]}
         BlueCurve {keys [<at 0 1>,<at 0.2382 1>,<at 1 1>]}
         GreenCurve {keys [<at 0 0.4865>,<at 0.2382 0.2891>,<at 1 0.1931>]}
         RedCurve {keys [<at 0 0.1198>,<at 0.2382 0.0521>,<at 1 0.1198>]}
        }}))
     V 2:Vector3f=&MakeVectorFromLinearColorRGB(Color=&ColorFromCurve(ColorCurve=DataInterface() {propertyValues {
         AlphaCurve {keys [<at 0 1>]}
         BlueCurve {keys [<at 0 1>,<at 0.2382 1>,<at 1 1>]}
         GreenCurve {keys [<at 0 0.4865>,<at 0.2382 0.2891>,<at 1 0.1931>]}
         RedCurve {keys [<at 0 0.1198>,<at 0.2382 0.0521>,<at 1 0.1198>]}
        }}))
     V 3:Vector3f=&MakeVectorFromLinearColorRGB(Color=&ColorFromCurve(ColorCurve=DataInterface() {propertyValues {
         AlphaCurve {keys [<at 0 1>]}
         BlueCurve {keys [<at 0 1>,<at 0.3338 1>]}
         GreenCurve {keys [<at 0 0.9749>,<at 0.3338 0.5897>]}
         RedCurve {keys [<at 0 0.9323>,<at 0.3338 0.0417>]}
        }}))
     V 4:Vector3f=&MakeVectorFromLinearColorRGB(Color=&ColorFromCurve(ColorCurve=DataInterface() {propertyValues {
         AlphaCurve {keys [<at 0 1 RCIM_Cubic>,<at 1 0 RCIM_Cubic>]}
         BlueCurve {keys [<at 0.3607 0.474 RCIM_Cubic>]}
         GreenCurve {keys [<at 0.3607 0.474 RCIM_Cubic>]}
         RedCurve {keys [<at 0.3607 0.474 RCIM_Cubic>]}
        }}))
     V 5:Vector3f=<1 1 1>
     Vector Count:ENiagara_MultipleLerpCount=::NewEnumerator2
    )
    ScaleA=true
    ScaleRGB=true
   )
   &ScaleMeshSize(Scale Factor=&VectorFromFloat(Value=&FloatFromCurve(
      FloatCurve=DataInterface() {propertyValues {Curve {keys [
          {arriveTangent=3, interpMode=RCIM_Cubic, leaveTangent=3, time=0, value=0.2}
          {arriveTangent=1, interpMode=RCIM_Cubic, leaveTangent=1, time=0.2, value=0.8}
          {tangentMode=RCTM_User, time=1, value=1.2}
         ]}}}
      Scale Curve=&MakeCustomFloatFromBool(Bool=$User.bIsFrontEndPreview, True Float=0.8, False Float=1)
     )))
   &DynamicMaterialParameters(
    Index 0 Param 1=&FloatFromCurve(FloatCurve=DataInterface() {propertyValues {Curve {keys [<at 0 1>,<at 1 1.2>]}}})
    Index 0 Param 2=&FloatFromCurve(FloatCurve=DataInterface() {propertyValues {Curve {keys [<at 0 10>,<at 0.1 5>,<at 0.5 0>]}}})
    Module.Param0WriteEnabled=true
    Module.Param1WriteEnabled=true
    Module.Param2WriteEnabled=true
    Module.Param3WriteEnabled=true
   )
   &ScaleColor(
    Scale Alpha=&FloatFromCurve(
     FloatCurve=DataInterface() {propertyValues {Curve {keys [
         {arriveTangent=-2.7142, interpMode=RCIM_Cubic, time=0, value=1}
         {arriveTangent=-1, interpMode=RCIM_Cubic, leaveTangent=-1, time=0.2852, value=0.1588}
         <at 1 0>
        ]}}}
     Scale Curve=&MakeCustomFloatFromBool(Bool=true, True Float=0.5, False Float=2)
    )
    Scale RGB=&VectorFromFloat(Value=&FloatFromCurve(FloatCurve=DataInterface() {propertyValues {Curve {keys [<at 0 10>,<at 0.5 1>,<at 1 0>]}}}))
    ScaleA=true
    ScaleRGB=true
   ) {module=ScaleColor001}
  ]
  rendererClasses [&NiagaraMeshRendererProperties]
  renderers [{rendererClass=&NiagaraMeshRendererProperties, rendererIndex=0}]
 }
 Impact_Flare {
  emitterUpdateScript [
   &EmitterState(Life Cycle Mode=::NewEnumerator1, Loop Behavior=::NewEnumerator1, Loop Delay=0.05, Loop Duration=0.5, UseLoopDelay=false)
   &SpawnBurst_Instantaneous(Spawn Count=1, Use Spawn Probability=false)
  ]
  particleSpawnScript [
   &InitializeParticle(
    Color=&SelectLinearColorFromArray(
     Array Sampling Mode=::NewEnumerator1
     Direct Array Index=&MakeIntFromFloat(Float=$User.VariantSwitch)
     Color Selection Array=DataInterface() {propertyValues {ColorData [<rgba 0.684 0.0521 1 1>, <rgba 0.684 0.0521 1 1>, <rgba 1 1 1 0>, <rgba 1 1 1 0>]}}
     Random Seed=?
    )
    Lifetime=0.45
    Mesh Scale Max=<1.4 1.4 1.4>
    Mesh Scale Min=<1.25 1.25 1.25>
    Mesh Scale Mode=::NewEnumerator2
    Sprite Size=<200 175>
    Sprite Size Mode=::NewEnumerator4
    UsePositionOffset=false
    Write Material Random=true
   ) {module=InitializeParticle001}
   &CameraOffset(Camera Offset Amount=&FN_FrontEndFloatModifier(Gameplay=30, Front End=0))
   &InitialMeshOrientation(
    Rotation=&RandomRangeVector(Minimum=<-1 -1 -1>, Maximum=<1 1 1>)
    Use Orientation Vector=true
    Use Rotation Vector=true
   ) {module=InitialMeshOrientation001}
  ]
  particleUpdateScript [
   &ParticleState()
   &ScaleColor(
    Scale Alpha=&FloatFromCurve(
     FloatCurve=DataInterface() {propertyValues {Curve {keys [
         {arriveTangent=-2.6501, interpMode=RCIM_Cubic, leaveTangent=-2.6501, tangentMode=RCTM_User, time=0, value=0}
         {arriveTangent=1.5, interpMode=RCIM_Cubic, leaveTangent=1.5, time=0.3, value=0}
         {arriveTangent=-1, interpMode=RCIM_Cubic, leaveTangent=-1, tangentMode=RCTM_Break, time=0.4, value=0.6}
         <at 1 0>
        ]}}}
     Scale Curve=3
    )
    Scale RGB=&LerpMultipleVectors(
     Alpha:NiagaraFloat=&Divide_Float(A=$User.VariantSwitch, B=4)
     V 1:Vector3f=&MakeVectorFromLinearColorRGB(Color=&ColorFromCurve(ColorCurve=DataInterface() {propertyValues {
         AlphaCurve {keys [<at 0 1 RCIM_Cubic>,<at 1 0 RCIM_Cubic>]}
         BlueCurve {keys [<at 0.3607 0.02 RCIM_Cubic>,<at 0.7561 2 RCIM_Cubic>]}
         GreenCurve {keys [<at 0.3607 0.0002 RCIM_Cubic>,<at 0.7561 0 RCIM_Cubic>]}
         RedCurve {keys [<at 0.3607 0.0051 RCIM_Cubic>,<at 0.7561 0.6667 RCIM_Cubic>]}
        }}))
     V 2:Vector3f=&MakeVectorFromLinearColorRGB(Color=&ColorFromCurve(ColorCurve=DataInterface() {propertyValues {
         AlphaCurve {keys [<at 0 1 RCIM_Cubic>,<at 1 0 RCIM_Cubic>]}
         BlueCurve {keys [<at 0.3607 0.02 RCIM_Cubic>,<at 0.7561 2 RCIM_Cubic>]}
         GreenCurve {keys [<at 0.3607 0.0013 RCIM_Cubic>,<at 0.7561 0.7441 RCIM_Cubic>]}
         RedCurve {keys [<at 0.3607 0 RCIM_Cubic>,<at 0.7561 0.0042 RCIM_Cubic>]}
        }}))
     V 3:Vector3f=&MakeVectorFromLinearColorRGB(Color=&ColorFromCurve(ColorCurve=DataInterface() {propertyValues {
         AlphaCurve {keys [<at 0 1 RCIM_Cubic>,<at 1 0 RCIM_Cubic>]}
         BlueCurve {keys [<at 0.41 0.0417 RCIM_Cubic>,<at 0.5973 0.8269>,<at 0.7665 2 RCIM_Cubic>]}
         GreenCurve {keys [<at 0.41 0.0119 RCIM_Cubic>,<at 0.5973 0.4484>,<at 0.7665 0.2385 RCIM_Cubic>]}
         RedCurve {keys [<at 0.41 0.0031 RCIM_Cubic>,<at 0.5973 0.0277>,<at 0.7665 0.7535 RCIM_Cubic>]}
        }}))
     V 4:Vector3f=&MakeVectorFromLinearColorRGB(Color=&ColorFromCurve(ColorCurve=DataInterface() {propertyValues {
         AlphaCurve {keys [<at 0 1 RCIM_Cubic>,<at 1 0 RCIM_Cubic>]}
         BlueCurve {keys [<at 0.3607 0.0012 RCIM_Cubic>,<at 0.7561 0.0597 RCIM_Cubic>]}
         GreenCurve {keys [<at 0.3607 0.0087 RCIM_Cubic>,<at 0.7561 0.5744 RCIM_Cubic>]}
         RedCurve {keys [<at 0.3607 0.02 RCIM_Cubic>,<at 0.7561 1 RCIM_Cubic>]}
        }}))
     V 5:Vector3f=&MakeVectorFromLinearColorRGB(Color=&ColorFromCurve(ColorCurve=DataInterface() {propertyValues {
         AlphaCurve {keys [<at 0 1 RCIM_Cubic>,<at 1 0 RCIM_Cubic>]}
         BlueCurve {keys [<at 0.4294 0.474>,<at 0.5753 0.8269>,<at 0.738 2 RCIM_Cubic>]}
         GreenCurve {keys [<at 0.4294 0.1428>,<at 0.5753 0.4484>,<at 0.738 0.2385 RCIM_Cubic>]}
         RedCurve {keys [<at 0.4294 0.0282>,<at 0.5753 0.0277>,<at 0.738 0.7535 RCIM_Cubic>]}
        }}))
     Vector Count:ENiagara_MultipleLerpCount=::NewEnumerator2
    )
    ScaleA=true
    ScaleRGB=true
   )
   &ScaleMeshSize(Scale Factor=&VectorFromFloat(Value=&FloatFromCurve(
      FloatCurve=DataInterface() {propertyValues {Curve {keys [
          {arriveTangent=1.5592, interpMode=RCIM_Cubic, leaveTangent=1.5592, tangentMode=RCTM_User, time=0, value=0}
          {arriveTangent=2.5, interpMode=RCIM_Cubic, leaveTangent=2.5, time=0.1, value=0}
          {arriveTangent=2.5, interpMode=RCIM_Cubic, leaveTangent=2.5, time=0.2, value=0.5}
          <at 0.6 1.25>
         ]}}}
      Scale Curve=&FN_FrontEndFloatModifier(Gameplay=$System.FX size scale, Front End=0.6)
     )))
   &DynamicMaterialParameters(
    Index 0 Param 1=&FloatFromCurve(FloatCurve=DataInterface() {propertyValues {Curve {keys [<at 0 1 RCIM_Cubic>,<at 0.8 0 RCIM_Cubic>]}}})
    Index 0 Param 2=2
    Index 0 Param 3=1
    Index 0 Param 4=1
    Index 1 Param 1=&FloatFromCurve(FloatCurve=DataInterface() {propertyValues {Curve {keys [
         {arriveTangent=-2.3163, interpMode=RCIM_Cubic, time=0, value=2}
         {arriveTangent=-1.9019, interpMode=RCIM_Cubic, leaveTangent=-1.9019, tangentMode=RCTM_User, time=1, value=0.4}
        ]}}})
    Index 1 Param 2=0.4
    Index 1 Param 3=0.6
    Index 1 Param 4=1
    Index 2 Param 1=&FloatFromCurve(FloatCurve=DataInterface() {propertyValues {Curve {keys [<at 0.46 0>,<at 1 1>]}}})
    Index 2 Param 2=45
    Module.Param0WriteEnabled=true
    Module.Param0WriteEnabled1=true
    Module.Param0WriteEnabled2=true
    Module.Param1WriteEnabled=true
    Module.Param1WriteEnabled1=true
    Module.Param1WriteEnabled2=true
    Module.Param2WriteEnabled=true
    Module.Param2WriteEnabled1=true
    Module.Param2WriteEnabled2=true
    Module.Param3WriteEnabled=true
    Module.Param3WriteEnabled1=true
    Module.Param3WriteEnabled2=true
    Write Parameter Index 1=true
    Write Parameter Index 2=true
   ) {module=DynamicMaterialParameters001}
   &ScaleColor(
    Scale Alpha=&FloatFromCurve(
     FloatCurve=DataInterface() {propertyValues {Curve {keys [
         {arriveTangent=-2.6501, interpMode=RCIM_Cubic, leaveTangent=-2.6501, tangentMode=RCTM_User, time=0, value=0}
         {arriveTangent=1.5, interpMode=RCIM_Cubic, leaveTangent=1.5, time=0.3, value=0}
         {arriveTangent=-1, interpMode=RCIM_Cubic, leaveTangent=-1, tangentMode=RCTM_Break, time=0.4, value=0.6}
         <at 1 0>
        ]}}}
     Scale Curve=3
    )
    Scale RGB=&Multiply_VectorByFloat(
     Vector=&MakeVectorFromLinearColorRGB(Color=&ColorFromCurve(ColorCurve=DataInterface() {propertyValues {
         AlphaCurve {keys [<at 0 1 RCIM_Cubic>,<at 1 0 RCIM_Cubic>]}
         BlueCurve {keys [<at 0.3607 0.02 RCIM_Cubic>,<at 0.7561 2 RCIM_Cubic>]}
         GreenCurve {keys [<at 0.3607 0.0002 RCIM_Cubic>,<at 0.7561 0 RCIM_Cubic>]}
         RedCurve {keys [<at 0.3607 0.0051 RCIM_Cubic>,<at 0.7561 0.6667 RCIM_Cubic>]}
        }}))
     Float=&FloatFromCurve(FloatCurve=DataInterface() {propertyValues {Curve {keys [{arriveTangent=-2.6501, interpMode=RCIM_Cubic, leaveTangent=-2.6501, tangentMode=RCTM_User, time=0, value=5}, <at 1 1>]}}})
    )
    ScaleA=true
    ScaleRGB=true
   ) {enabled=false, module=ScaleColor001}
  ]
  rendererClasses [&NiagaraMeshRendererProperties]
  renderers [{rendererClass=&NiagaraMeshRendererProperties, rendererIndex=0}]
 }
 RingColor002 {
  emitterUpdateScript [
   &EmitterState(
    Inactive Response=::NewEnumerator1
    Life Cycle Mode=::NewEnumerator1
    Loop Behavior=::NewEnumerator1
    Loop Duration=1.2334
    UseLoopDelay=true
   )
   &SpawnBurst_Instantaneous(Use Spawn Probability=false, UseLoopCountLimit=false)
   &SpawnBurst_Instantaneous(Spawn Time=1.2333, Use Spawn Probability=false, UseLoopCountLimit=false) {module=SpawnBurst_Instantaneous001}
  ]
  particleSpawnScript [
   &PartitionParticles()
   &InitializeParticle(Lifetime=0.35, Mesh Scale Mode=::NewEnumerator4, UsePositionOffset=true, Write Material Random=true)
   &InitialMeshOrientation(Mesh Orientation Mode=::NewEnumerator7, Rotation=<0.26 0 0>, Use Rotation Vector=true)
   &CameraOffset(Camera Offset Amount=&FN_FrontEndFloatModifier(Gameplay=50, Front End=0))
   /SetVariables(Particles.Random=&RandomRangeFloat(Minimum=1, Maximum=1)) {isSetParameters=true, module=SetVariables_#3}
   &InitialMeshOrientation(Mesh Orientation Mode=::NewEnumerator7, Rotation=<-0.05 0.07 0>, Use Rotation Vector=true) {module=InitialMeshOrientation001}
   &InitialMeshOrientation(Mesh Orientation Mode=::NewEnumerator7, Use Rotation Vector=true) {enabled=false, module=InitialMeshOrientation002}
  ]
  particleUpdateScript [
   &ParticleState()
   &UpdateMeshOrientation(Rotation Rate=3, Rotation Vector=<0 1 0>)
   &ScaleMeshSize(Scale Factor=&Multiply_VectorByFloat(
     Vector=<1 1 1>
     Float=&Multiply_Float(
      A=0.5
      B=&FloatFromCurve(
       FloatCurve=DataInterface() {propertyValues {Curve {keys [<at 0 0.5>,<at 1 1.5>]}}}
       Scale Curve=&Multiply_Float(A=1.7, B=$System.ScaleSize)
      )
     )
    ))
   &DynamicMaterialParameters(
    Index 0 Param 1=&FloatFromCurve(
     FloatCurve=DataInterface() {propertyValues {Curve {keys [<at 0 1>,<at 1 0.7>]}}}
     Scale Curve=&RandomRangeFloat(Minimum=1, Maximum=1)
    )
    Index 0 Param 2=0.3
    Index 0 Param 3=3
    Index 0 Param 4=&FloatFromCurve(FloatCurve=DataInterface() {propertyValues {Curve {keys [<at 0 1>,<at 0.7 1>,<at 1 0>]}}}, Scale Curve=5)
    Module.Param0WriteEnabled=true
    Module.Param1WriteEnabled=true
    Module.Param2WriteEnabled=true
    Module.Param3WriteEnabled=true
   ) {module=DynamicMaterialParameters001}
   &ScaleColor(
    Scale Alpha=&FloatFromCurve(
     FloatCurve=DataInterface() {propertyValues {Curve {keys [<at 0 3 RCIM_Cubic>, {arriveTangent=-3, interpMode=RCIM_Cubic, leaveTangent=-3, time=0.6258, value=0.8384}, <at 1 0>]}}}
     Scale Curve=2
    )
    Scale RGB=&Multiply_Vector(
     A=&LerpMultipleVectors(
      Alpha:NiagaraFloat=&Divide_Float(A=$User.VariantSwitch, B=4)
      V 1:Vector3f=&MakeVectorFromLinearColorRGB(Color=&ColorFromCurve(ColorCurve=DataInterface() {propertyValues {
          AlphaCurve {keys [<at 0 1 RCIM_Cubic>,<at 1 0 RCIM_Cubic>]}
          BlueCurve {keys [<at 0.2333 2>, {arriveTangent=-0.6791, interpMode=RCIM_Cubic, leaveTangent=-0.6791, time=0.624, value=1.5}, <at 0.9695 1.5>]}
          GreenCurve {keys [<at 0.2333 0.068>, {arriveTangent=0.166, interpMode=RCIM_Cubic, leaveTangent=0.166, time=0.624, value=0.1562}, <at 0.9695 0.1902>]}
          RedCurve {keys [<at 0.2333 0.1615>, {arriveTangent=0.2279, interpMode=RCIM_Cubic, leaveTangent=0.2279, time=0.624, value=0.9401}, <at 0.9695 0.3293>]}
         }}))
      V 2:Vector3f=&MakeVectorFromLinearColorRGB(Color=&ColorFromCurve(ColorCurve=DataInterface() {propertyValues {
          AlphaCurve {keys [<at 0 1 RCIM_Cubic>,<at 1 0 RCIM_Cubic>]}
          BlueCurve {keys [<at 0.264 1>, {arriveTangent=0.1218, interpMode=RCIM_Cubic, leaveTangent=0.1218, time=0.624, value=1}, <at 0.9695 1.0859>]}
          GreenCurve {keys [<at 0.264 0.1067>, {arriveTangent=-0.0315, interpMode=RCIM_Cubic, leaveTangent=-0.0315, time=0.624, value=0.398}, <at 0.9695 0.0845>]}
          RedCurve {keys [<at 0.264 0.0072>, {arriveTangent=0.1242, interpMode=RCIM_Cubic, leaveTangent=0.1242, time=0.624, value=0.0774}, <at 0.9695 0.0948>]}
         }}))
      V 3:Vector3f=&MakeVectorFromLinearColorRGB(Color=&ColorFromCurve(ColorCurve=DataInterface() {propertyValues {
          AlphaCurve {keys [<at 0 1 RCIM_Cubic>,<at 1 0 RCIM_Cubic>]}
          BlueCurve {keys [<at 0.2657 0.4531>,<at 0.8011 1 RCIM_Cubic>,<at 0.9488 1.0859>]}
          GreenCurve {keys [<at 0.2657 0.2697>,<at 0.8011 0.0957 RCIM_Cubic>,<at 0.9488 0.2617>]}
          RedCurve {keys [<at 0.2657 0.0214>,<at 0.8011 0.3943 RCIM_Cubic>,<at 0.9488 0.0848>]}
         }}))
      V 4:Vector3f=&MakeVectorFromLinearColorRGB(Color=&ColorFromCurve(ColorCurve=DataInterface() {propertyValues {
          AlphaCurve {keys [<at 0 1 RCIM_Cubic>,<at 1 0 RCIM_Cubic>]}
          BlueCurve {keys [<at 0.264 0.0689>, {arriveTangent=-0.0178, interpMode=RCIM_Cubic, leaveTangent=-0.0178, time=0.624, value=0.0926}, <at 0.9695 0.0564>]}
          GreenCurve {keys [<at 0.264 0.4427>, {arriveTangent=0.14, interpMode=RCIM_Cubic, leaveTangent=0.14, time=0.624, value=0.5756}, <at 0.9695 0.5414>]}
          RedCurve {keys [<at 0.264 1>, {arriveTangent=0.1218, interpMode=RCIM_Cubic, leaveTangent=0.1218, time=0.624, value=1}, <at 0.9695 1.0859>]}
         }}))
      V 5:Vector3f=&MakeVectorFromLinearColorRGB(Color=&ColorFromCurve(ColorCurve=DataInterface() {propertyValues {
          AlphaCurve {keys [<at 0 1 RCIM_Cubic>,<at 1 0 RCIM_Cubic>]}
          BlueCurve {keys [<at 0.2908 0.8438>,<at 0.5848 1.4727>,<at 0.7774 1.5>]}
          GreenCurve {keys [<at 0.2908 0.5607>,<at 0.5848 0.4092>,<at 0.7774 0.7545>]}
          RedCurve {keys [<at 0.2908 0.1719>,<at 0.5848 1.5>,<at 0.7774 0.1936>]}
         }}))
      Vector Count:ENiagara_MultipleLerpCount=::NewEnumerator2
     )
     B=&VectorFromFloat(Value=&FloatFromCurve(
       FloatCurve=DataInterface() {propertyValues {Curve {keys [<at 0 10>,<at 1 1>]}}}
       Scale Curve=&SelectFloatFromArray(
        Direct Array Index=&MakeIntFromFloat(Float=$User.VariantSwitch)
        Float Selection Array=DataInterface() {propertyValues {FloatData [2,2,1.2,1,1.5,1]}}
       )
      ))
    )
    ScaleA=true
    ScaleRGB=true
   )
  ]
  rendererClasses [&NiagaraMeshRendererProperties]
  renderers [{rendererClass=&NiagaraMeshRendererProperties, rendererIndex=0}]
 }
 RingColor02001 {
  emitterUpdateScript [
   &EmitterState(Inactive Response=::NewEnumerator1, Life Cycle Mode=::NewEnumerator1, Loop Behavior=::NewEnumerator1, Loop Duration=1.207, UseLoopDelay=true)
   &SpawnBurst_Instantaneous(Use Spawn Probability=false, UseLoopCountLimit=false)
  ]
  particleSpawnScript [
   &PartitionParticles()
   &InitializeParticle(Lifetime=0.4, Mesh Scale Mode=::NewEnumerator4, UsePositionOffset=true, Write Material Random=true)
   &InitialMeshOrientation(Mesh Orientation Mode=::NewEnumerator7, Rotation=<-0.25 0 0>, Use Rotation Vector=true)
   &CameraOffset(Camera Offset Amount=&FN_FrontEndFloatModifier(Gameplay=50, Front End=0))
   /SetVariables(Particles.Random=&RandomRangeFloat(Minimum=1, Maximum=1)) {isSetParameters=true, module=SetVariables_#3}
   &InitialMeshOrientation(Mesh Orientation Mode=::NewEnumerator7, Rotation=<0.05 -0.09 0>, Use Rotation Vector=true) {module=InitialMeshOrientation001}
   &InitialMeshOrientation(Mesh Orientation Mode=::NewEnumerator7, Use Rotation Vector=true) {enabled=false, module=InitialMeshOrientation002}
  ]
  particleUpdateScript [
   &ParticleState()
   &UpdateMeshOrientation(Rotation Rate=2, Rotation Vector=<0 1 0>)
   &ScaleMeshSize(Scale Factor=&Multiply_VectorByFloat(
     Vector=<1 1 1>
     Float=&Multiply_Float(
      A=0.5
      B=&FloatFromCurve(
       FloatCurve=DataInterface() {propertyValues {Curve {keys [<at 0 0.5>,<at 1 1.5>]}}}
       Scale Curve=&Multiply_Float(A=1.7, B=$System.ScaleSize)
      )
     )
    ))
   &ScaleColor(
    Scale Alpha=&FloatFromCurve(
     FloatCurve=DataInterface() {propertyValues {Curve {keys [<at 0 3 RCIM_Cubic>, {arriveTangent=-3, interpMode=RCIM_Cubic, leaveTangent=-3, time=0.6258, value=0.8384}, <at 1 0>]}}}
     Scale Curve=2
    )
    Scale RGB=&Multiply_Vector(
     A=&LerpMultipleVectors(
      Alpha:NiagaraFloat=&Divide_Float(A=$User.VariantSwitch, B=4)
      V 1:Vector3f=&MakeVectorFromLinearColorRGB(Color=&ColorFromCurve(ColorCurve=DataInterface() {propertyValues {
          AlphaCurve {keys [<at 0 1 RCIM_Cubic>,<at 1 0 RCIM_Cubic>]}
          BlueCurve {keys [<at 0.2333 2>, {arriveTangent=-0.6791, interpMode=RCIM_Cubic, leaveTangent=-0.6791, time=0.624, value=1.5}, <at 0.9695 1.5>]}
          GreenCurve {keys [<at 0.2333 0.068>, {arriveTangent=0.166, interpMode=RCIM_Cubic, leaveTangent=0.166, time=0.624, value=0.1562}, <at 0.9695 0.1902>]}
          RedCurve {keys [<at 0.2333 0.1615>, {arriveTangent=0.2279, interpMode=RCIM_Cubic, leaveTangent=0.2279, time=0.624, value=0.9401}, <at 0.9695 0.3293>]}
         }}))
      V 2:Vector3f=&MakeVectorFromLinearColorRGB(Color=&ColorFromCurve(ColorCurve=DataInterface() {propertyValues {
          AlphaCurve {keys [<at 0 1 RCIM_Cubic>,<at 1 0 RCIM_Cubic>]}
          BlueCurve {keys [<at 0.264 1>, {arriveTangent=0.1218, interpMode=RCIM_Cubic, leaveTangent=0.1218, time=0.624, value=1.5}, <at 0.9695 1.0859>]}
          GreenCurve {keys [<at 0.264 0.1067>, {arriveTangent=-0.0315, interpMode=RCIM_Cubic, leaveTangent=-0.0315, time=0.624, value=0.597}, <at 0.9695 0.0845>]}
          RedCurve {keys [<at 0.264 0.0072>, {arriveTangent=0.1242, interpMode=RCIM_Cubic, leaveTangent=0.1242, time=0.624, value=0.1161}, <at 0.9695 0.0948>]}
         }}))
      V 3:Vector3f=&MakeVectorFromLinearColorRGB(Color=&ColorFromCurve(ColorCurve=DataInterface() {propertyValues {
          AlphaCurve {keys [<at 0 1 RCIM_Cubic>,<at 1 0 RCIM_Cubic>]}
          BlueCurve {keys [<at 0.3567 0.3646>,<at 0.6785 1.0859>,<at 0.9461 1.0859>]}
          GreenCurve {keys [<at 0.3567 0.1643>,<at 0.6785 0.1165>,<at 0.9461 0.5001>]}
          RedCurve {keys [<at 0.3567 0.0217>,<at 0.6785 0.4449>,<at 0.9461 0.0886>]}
         }}))
      V 4:Vector3f=&MakeVectorFromLinearColorRGB(Color=&ColorFromCurve(ColorCurve=DataInterface() {propertyValues {
          AlphaCurve {keys [<at 0 1 RCIM_Cubic>,<at 1 0 RCIM_Cubic>]}
          BlueCurve {keys [<at 0.264 0.0743>, {arriveTangent=-0.0253, interpMode=RCIM_Cubic, leaveTangent=-0.0253, time=0.624, value=0.1347}, <at 0.9695 0.0564>]}
          GreenCurve {keys [<at 0.264 0.4799>, {arriveTangent=0.0873, interpMode=RCIM_Cubic, leaveTangent=0.0873, time=0.624, value=0.6206}, <at 0.9695 0.5414>]}
          RedCurve {keys [<at 0.264 1>, {arriveTangent=0.1218, interpMode=RCIM_Cubic, leaveTangent=0.1218, time=0.624, value=1}, <at 0.9695 1.0859>]}
         }}))
      V 5:Vector3f=&MakeVectorFromLinearColorRGB(Color=&ColorFromCurve(ColorCurve=DataInterface() {propertyValues {
          AlphaCurve {keys [<at 0 1 RCIM_Cubic>,<at 1 0 RCIM_Cubic>]}
          BlueCurve {keys [<at 0.2908 0.8438>,<at 0.5848 1.4727>,<at 0.7774 1.5>]}
          GreenCurve {keys [<at 0.2908 0.5607>,<at 0.5848 0.4092>,<at 0.7774 0.7545>]}
          RedCurve {keys [<at 0.2908 0.1719>,<at 0.5848 1.5>,<at 0.7774 0.1936>]}
         }}))
      Vector Count:ENiagara_MultipleLerpCount=::NewEnumerator2
     )
     B=&VectorFromFloat(Value=&FloatFromCurve(
       FloatCurve=DataInterface() {propertyValues {Curve {keys [<at 0 10>,<at 1 1>]}}}
       Scale Curve=&SelectFloatFromArray(
        Direct Array Index=&MakeIntFromFloat(Float=$User.VariantSwitch)
        Float Selection Array=DataInterface() {propertyValues {FloatData [2,2,1.2,1,1.5,1]}}
       )
      ))
    )
    ScaleA=true
    ScaleRGB=true
   )
   &ScaleColor(
    Scale Alpha=&FloatFromCurve(
     FloatCurve=DataInterface() {propertyValues {Curve {keys [<at 0 3 RCIM_Cubic>, {arriveTangent=-3, interpMode=RCIM_Cubic, leaveTangent=-3, time=0.6258, value=0.8384}, <at 1 0>]}}}
     Scale Curve=2
    )
    Scale RGB=&Multiply_VectorByFloat(
     Vector=&MakeVectorFromLinearColorRGB(Color=&ColorFromCurve(ColorCurve=DataInterface() {propertyValues {
         AlphaCurve {keys [<at 0 1 RCIM_Cubic>,<at 1 0 RCIM_Cubic>]}
         BlueCurve {keys [<at 0.2332 2>, {arriveTangent=-0.6791, interpMode=RCIM_Cubic, leaveTangent=-0.6791, time=0.624, value=1.5}, <at 0.9695 1.5>]}
         GreenCurve {keys [<at 0.2332 0.068>, {arriveTangent=0.166, interpMode=RCIM_Cubic, leaveTangent=0.166, time=0.624, value=0.1563}, <at 0.9695 0.1902>]}
         RedCurve {keys [<at 0.2332 0.1615>, {arriveTangent=0.2279, interpMode=RCIM_Cubic, leaveTangent=0.2279, time=0.624, value=0.9401}, <at 0.9695 0.3293>]}
        }}))
     Float=&FloatFromCurve(FloatCurve=DataInterface() {propertyValues {Curve {keys [<at 0 10>,<at 1 1>]}}}, Scale Curve=2)
    )
    ScaleA=true
    ScaleRGB=true
   ) {enabled=false, module=ScaleColor001}
   &DynamicMaterialParameters(
    Index 0 Param 1=&FloatFromCurve(
     FloatCurve=DataInterface() {propertyValues {Curve {keys [<at 0 1>,<at 1 0.7>]}}}
     Scale Curve=&RandomRangeFloat(Minimum=0.5, Maximum=1)
    )
    Index 0 Param 2=0.3
    Index 0 Param 3=3
    Index 0 Param 4=&FloatFromCurve(FloatCurve=DataInterface() {propertyValues {Curve {keys [<at 0 1>,<at 0.7 1>,<at 1 0>]}}}, Scale Curve=5)
    Module.Param0WriteEnabled=true
    Module.Param1WriteEnabled=true
    Module.Param2WriteEnabled=true
    Module.Param3WriteEnabled=true
   ) {module=DynamicMaterialParameters001}
  ]
  rendererClasses [&NiagaraMeshRendererProperties]
  renderers [{rendererClass=&NiagaraMeshRendererProperties, rendererIndex=0}]
 }
 Spark {
  emitterUpdateScript [
   &EmitterState(Life Cycle Mode=::NewEnumerator1, Loop Behavior=::NewEnumerator1, Loop Delay=0.05, Loop Duration=1, UseLoopDelay=true)
   &SpawnBurst_Instantaneous(Spawn Count=30, Use Spawn Probability=false)
  ]
  particleSpawnScript [
   &InitializeParticle(
    Color=<rgba 2.9213 1.0815 10 1>
    Lifetime Max=1.4
    Lifetime Min=0.7
    Lifetime Mode=::NewEnumerator1
    Sprite Size Max=<2 13>
    Sprite Size Min=<3 10>
    Sprite Size Mode=::NewEnumerator2
    UsePositionOffset=false
    Write Material Random=true
   )
   &AddVelocityFromPoint(
    Module.Use Falloff Distance=false
    Velocity Strength=&Multiply_Float(
     A=&RandomRangeFloat(Minimum=400, Maximum=1000)
     B=&FN_FrontEndFloatModifier(Gameplay=1, Front End=0.5)
    )
   )
   /SetVariables() {isSetParameters=true}
  ]
  particleUpdateScript [
   &ParticleState()
   &ScaleColor(
    Scale RGB=&MakeVectorFromLinearColorRGB(Color=&SelectLinearColorFromArray(
      Array Sampling Mode=::NewEnumerator1
      Direct Array Index=&MakeIntFromFloat(Float=$User.VariantSwitch)
      Color Selection Array=DataInterface() {propertyValues {ColorData [<rgba 10 0 15 1>, <rgba 0.1107 5.4372 15 1>, <rgba 0.1107 5.4372 15 1>, <rgba 15 9.3743 0.6297 1>, <rgba 1.1858 6.3538 10 1>]}}
      Random Seed=?
     ))
    ScaleA=true
    ScaleRGB=true
   )
   &ScaleColor(
    Scale Alpha=&FNTimeOfDayScalar(Night Scale=1, Morning Scale=1, Evening Scale=1)
    Scale RGB=&VectorFromFloat(Value=10)
    ScaleA=true
    ScaleRGB=true
   ) {module=ScaleColor001}
   &Drag(Drag=&RandomRangeFloat(Minimum=8, Maximum=15), Use Linear Drag=true, Use Rotational Drag=false)
   &CurlNoiseForce(Noise Strength=&RandomRangeFloat(Minimum=200, Maximum=400), Pan Noise=true)
   &GravityForce(Gravity=&RandomRangeVector(Minimum=<0 0 0>, Maximum=<0 0 -100>))
   &SolveForcesAndVelocity(Clamp Velocity=false, Limit Acceleration=false)
   &ScaleSpriteSizeBySpeed(Max Scale Factor=<1 5>, Min Scale Factor=<0.5 0.1>)
   &ScaleSpriteSize(Scale Factor=&MakeVector2D(
     X=&FloatFromCurve(
      FloatCurve=DataInterface() {propertyValues {Curve {keys [<at 0 0.5>,<at 0.1 1>,<at 1 0.5>]}}}
      Scale Curve=&FN_FrontEndFloatModifier(Gameplay=1, Front End=0.4)
     )
     Y=&FloatFromCurve(
      FloatCurve=DataInterface() {propertyValues {Curve {keys [<at 0 1>,<at 1 0>]}}}
      Scale Curve=&FN_FrontEndFloatModifier(Gameplay=1, Front End=0.4)
     )
    ))
   /SetVariables(Particles.Glow03_Color=&MakeLinearColorFromVectorAndFloat(
     "Vector (RGB)"=&MakeVectorFromLinearColorRGB(Color=&MultiplyLinearColor(
       Linear Color=&SelectLinearColorFromArray(
        Array Sampling Mode=::NewEnumerator1
        Direct Array Index=&MakeIntFromFloat(Float=$User.VariantSwitch)
        Color Selection Array=DataInterface() {propertyValues {ColorData [<rgba 0.5111 0.0795 1 0.2>, <rgba 0 0.2055 1 0.2>, <rgba 0.2154 0.0542 1 0.2>, <rgba 1 0.3259 0.037 0.2>, <rgba 0.6595 0.2999 1 0.2>]}}
        Random Seed=?
       )
       Scale Factor=&MakeVector4_FromFloat(Float=1)
      ))
     "Float (Alpha)"=&Multiply_Float(A=0.2, B=1)
    )) {isSetParameters=true}
   /SetVariables(Particles.Glow03_Size=&Multiply_Vector2D_ByFloat(
     Float=&FloatFromCurve(
      FloatCurve=DataInterface() {propertyValues {Curve {keys [
          <at 0 1 RCIM_Cubic>
          {arriveTangent=-1, interpMode=RCIM_Cubic, leaveTangent=-1, time=0.2, value=1.3}
          {arriveTangent=-0.1423, interpMode=RCIM_Cubic, leaveTangent=-0.1423, tangentMode=RCTM_User, time=1, value=0}
         ]}}}
      Scale Curve=15
     )
     Vector2D=<0.5 0.8>
    )) {isSetParameters=true}
  ]
  rendererClasses [&NiagaraSpriteRendererProperties]
  renderers [
   {rendererClass=&NiagaraSpriteRendererProperties}
   {rendererClass=&NiagaraSpriteRendererProperties, rendererIndex=1}
  ]
 }
 SphereMesh_Colorful {
  emitterUpdateScript [
   &EmitterState(
    Inactive Response=::NewEnumerator1
    Life Cycle Mode=::NewEnumerator1
    Loop Behavior=::NewEnumerator1
    Loop Delay=0.07
    Loop Duration=1
    UseLoopDelay=true
   )
   &SpawnBurst_Instantaneous(Use Spawn Probability=false, UseLoopCountLimit=false)
  ]
  particleSpawnScript [
   &InitializeParticle(Lifetime=0.3, Mesh Scale=<0.01 1 1>, Mesh Scale Mode=::NewEnumerator4, UsePositionOffset=false, Write Material Random=true)
   &InitialMeshOrientation(Mesh Orientation Mode=::NewEnumerator2, Use Rotation Vector=true)
   &CameraOffset(Camera Offset Amount=30)
  ]
  particleUpdateScript [
   &ParticleState()
   &ScaleMeshSize(Scale Factor=&Multiply_VectorByFloat(
     Vector=<1 1 1>
     Float=&FloatFromCurve(
      FloatCurve=DataInterface() {propertyValues {Curve {keys [<at 0 1>,<at 1 0>]}}}
      Scale Curve=&MakeCustomFloatFromBool(Bool=$User.bIsFrontEndPreview, True Float=4, False Float=5)
     )
    ))
   &DynamicMaterialParameters(Module.Index0Param0Write=true, Module.Index0Param1Write=true, Module.Index0Param2Write=true, Module.Index0Param3Write=true)
   /SetVariables(Particles.VisibilityTag=&SelectIntFromArray(
     Direct Array Index=&MakeIntFromFloat(Float=$User.VariantSwitch)
     Int Selection Array=DataInterface() {propertyValues {IntData [0,1,2,3,4,5]}}
    )) {isSetParameters=true}
   &ScaleColor(
    Scale Alpha=&FloatFromCurve(
     FloatCurve=DataInterface() {propertyValues {Curve {keys [<at 0 1>,<at 0.5 1>,<at 1 0>]}}}
     Scale Curve=&SelectFloatFromArray(
      Direct Array Index=&MakeIntFromFloat(Float=$User.VariantSwitch)
      Float Selection Array=DataInterface() {propertyValues {FloatData [1,1,0.5]}}
     )
    )
    Scale RGB=&Multiply_VectorByFloat(
     Vector=<1 1 1>
     Float=&FloatFromCurve(FloatCurve=DataInterface() {propertyValues {Curve {keys [<at 0 1 RCIM_Cubic>,<at 0.7 0>]}}})
    )
    ScaleA=true
    ScaleRGB=true
   )
  ]
  rendererClasses [&NiagaraMeshRendererProperties]
  renderers [
   {rendererClass=&NiagaraMeshRendererProperties, rendererIndex=0}
   {rendererClass=&NiagaraMeshRendererProperties, rendererIndex=1}
   {rendererClass=&NiagaraMeshRendererProperties, rendererIndex=2}
   {rendererClass=&NiagaraMeshRendererProperties, rendererIndex=3}
   {rendererClass=&NiagaraMeshRendererProperties, rendererIndex=4}
  ]
 }
}
summary {
 systemName=NS_Pickaxe_TidalNinja_Impact
 emitters [
  {emitterName=Impact_Flare, rendererClasses [&NiagaraMeshRendererProperties]}
  {emitterName=Spark, rendererClasses [&NiagaraSpriteRendererProperties]}
  {emitterName=Dark_glow, rendererClasses [&NiagaraSpriteRendererProperties]}
  {emitterName=SphereMesh_Colorful, rendererClasses [&NiagaraMeshRendererProperties]}
  {emitterName=ExplodeFresnelSphere, rendererClasses [&NiagaraMeshRendererProperties]}
  {emitterName=RingColor002, rendererClasses [&NiagaraMeshRendererProperties]}
  {emitterName=RingColor02001, rendererClasses [&NiagaraMeshRendererProperties]}
 ]
 userVariables [
  User.bIsFrontEnd {type=NiagaraBool}
  User.bIsFrontEndPreview {type=NiagaraBool}
  User.VariantSwitch {type=NiagaraFloat, defaultValue=0, description=""}
 ]
}
