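# Niagara swing effect for the TidalNinja pickaxe, layering animation-trail ribbons with sprite sparks on the system's skeletal mesh.
# User.VariantSwitch indexes arrays that toggle black, Galaxy and Holifoil trail layers and pick their colors, while two always-spawned energy trails blend color curves by the same value.
# Sparks spawn per unit moved, inherit velocity and scatter through curl noise, gravity and drag.
# bIsFrontEndPreview lowers brightness or alpha on several trails.
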NiagaraSystem /BRCosmetics/Weapons/FORT_Melee/Pickaxe_TidalNinja/FX/NS_Pickaxe_TidalNinja_Swing
dependencies {
 usedDataInterfaces [
  &NiagaraDataInterfaceColorCurve
  &NiagaraDataInterfaceCurve
  &NiagaraDataInterfaceArrayInt32
  &NiagaraDataInterfaceArrayColor
  &NiagaraDataInterfaceArrayFloat
  &NiagaraDataInterfaceSkeletalMesh
 ]
 usedDynamicInputs []
 usedModules [
  &self:SystemSpawnScript.SystemScriptSource.SystemScriptGraph.NiagaraNodeAssignment_0.SetVariables_#1
  &SystemState
 ]
}
emitters {
 BlackTrail {
  emitterUpdateScript [
   &EmitterState(
    Inactive Response=::NewEnumerator1
    Life Cycle Mode=::NewEnumerator1
    Loop Behavior=::NewEnumerator1
    Loop Delay=0.1
    Loop Duration=0.7
    UseLoopDelay=true
   )
   &FN_SpawnAnimTrail(Number of Anim Trails:NiagaraInt32=&SelectIntFromArray(
     Direct Array Index=&MakeIntFromFloat(Float=$User.VariantSwitch)
     Int Selection Array=DataInterface() {propertyValues {IntData [1,1,0,1,0,1]}}
    ))
  ]
  particleSpawnScript [
   &FN_InitializeAnimTrailBoneIndecies()
   &FN_InitializeAnimTrail(AnimTrailWidth=2, Color=<rgba 5 5 5 0.5>, Lifetime=0.4, SkeletalMesh=$System.Skeletal Mesh)
  ]
  particleUpdateScript [
   &DynamicMaterialParameters(
    Index 1 Param 1=&FloatFromCurve(FloatCurve=DataInterface() {propertyValues {Curve {keys [{tangentMode=RCTM_User, time=0, value=1}, {tangentMode=RCTM_User, time=1, value=1}]}}})
    Index 1 Param 3=&FloatFromCurve(FloatCurve=DataInterface() {propertyValues {Curve {keys [<at 0 -1 RCIM_Cubic>, <at 0.55 0.6116>, {arriveTangent=-1.3527, tangentMode=RCTM_User, time=1, value=1}]}}})
    Module.Index0Param0Write=true
    Module.Index0Param1Write=true
    Module.Index0Param2Write=false
    Module.Index0Param3Write=false
    Module.Index1Param0Write=true
    Module.Index1Param1Write=false
    Module.Index1Param2Write=true
    Module.Index1Param3Write=false
    Write Parameter Index 1=true
   )
   &ParticleState()
   &FN_UpdateAnimTrail(SkeletalMesh=$System.Skeletal Mesh)
   &ScaleColor(
    Scale Alpha=1.5
    Scale RGB=&MakeVectorFromLinearColorRGB(Color=&SelectLinearColorFromArray(
      Array Sampling Mode=::NewEnumerator1
      Direct Array Index=&MakeIntFromFloat(Float=$User.VariantSwitch)
      Color Selection Array=DataInterface() {propertyValues {ColorData [<rgba 0.3699 0.0488 1 2>, <rgba 0.0724 0.257 1 2>, <rgba 0.9351 0.901 1 1>, <rgba 0.401 0.1788 0.0329 2>, <rgba 0.1011 0.2951 0.4219 2>]}}
      Random Seed=?
     ))
    ScaleA=true
    ScaleRGB=true
   ) {module=ScaleColor001}
   &ScaleColor(
    Scale Alpha=&FloatFromCurve(
     FloatCurve=DataInterface() {propertyValues {Curve {keys [
         {arriveTangent=-1.0897, interpMode=RCIM_Cubic, leaveTangent=-1.0897, tangentMode=RCTM_User, time=0, value=1}
         {arriveTangent=-1.9706, interpMode=RCIM_Cubic, leaveTangent=-1.9706, tangentMode=RCTM_User, time=1, value=0}
        ]}}}
     Scale Curve=&FloatFromCurve(
      FloatCurve=DataInterface() {propertyValues {Curve {keys [
          <at 0 1>
          <at 0.2 1>
          {arriveTangent=-5.793, interpMode=RCIM_Cubic, leaveTangent=-5.793, tangentMode=RCTM_User, time=0.3, value=1}
          <at 0.8 0>
         ]}}}
      CurveIndex=$Emitter.Age
      Scale Curve=1.3
     )
    )
    Scale RGB=&Multiply_VectorByFloat(
     Vector=<1 1 1>
     Float=&FloatFromCurve(FloatCurve=DataInterface() {propertyValues {Curve {keys [<at 0 1>,<at 0.35 0.1>]}}})
    )
    ScaleA=true
    ScaleRGB=true
   )
   &ScaleRibbonWidth(Ribbon Width Scale=&FloatFromCurve(FloatCurve=DataInterface() {propertyValues {Curve {
        defaultValue=1
        keys [
         {interpMode=RCIM_Cubic, tangentMode=RCTM_User, time=0, value=1}
         {arriveTangent=-2.8868, interpMode=RCIM_Cubic, leaveTangent=-2.8868, tangentMode=RCTM_User, time=1, value=0}
        ]
       }}}))
  ]
  rendererClasses [&NiagaraRibbonRendererProperties]
  renderers [{rendererClass=&NiagaraRibbonRendererProperties}]
 }
 BlackTrail001 {
  emitterUpdateScript [
   &EmitterState(
    Inactive Response=::NewEnumerator1
    Life Cycle Mode=::NewEnumerator1
    Loop Behavior=::NewEnumerator1
    Loop Delay=0.1
    Loop Duration=0.7
    UseLoopDelay=true
   )
   &FN_SpawnAnimTrail(Number of Anim Trails:NiagaraInt32=&SelectIntFromArray(
     Direct Array Index=&MakeIntFromFloat(Float=$User.VariantSwitch)
     Int Selection Array=DataInterface() {propertyValues {IntData [1,1,0,0,0,0]}}
    ))
  ]
  particleSpawnScript [
   &FN_InitializeAnimTrailBoneIndecies()
   &FN_InitializeAnimTrail(Lifetime=0.3, SkeletalMesh=$System.Skeletal Mesh)
  ]
  particleUpdateScript [
   &ParticleState()
   &FN_UpdateAnimTrail(SkeletalMesh=$System.Skeletal Mesh)
   &ScaleColor(
    Scale RGB=&MakeVectorFromLinearColorRGB(Color=&SelectLinearColorFromArray(
      Array Sampling Mode=::NewEnumerator1
      Direct Array Index=&MakeIntFromFloat(Float=$User.VariantSwitch)
      Color Selection Array=DataInterface() {propertyValues {ColorData [<rgba 19.19 3.7899 30 1>, <rgba 0.205 14.0687 30 1>]}}
      Random Seed=?
     ))
    ScaleA=true
    ScaleRGB=true
   ) {module=ScaleColor001}
   &ScaleColor(
    Scale Alpha=&Multiply_Float(
     A=&FloatFromCurve(FloatCurve=DataInterface() {propertyValues {Curve {keys [
          {interpMode=RCIM_Cubic, leaveTangent=-2.3646, tangentMode=RCTM_Break, time=0, value=1}
          {arriveTangent=-0.4446, interpMode=RCIM_Cubic, tangentMode=RCTM_Break, time=1, value=0}
         ]}}})
     B=&MakeCustomFloatFromBool(Bool=$User.bIsFrontEndPreview, True Float=1, False Float=2)
    )
    Scale RGB=&Multiply_Vector(
     A=<1 1 1>
     B=&VectorFromFloat(Value=&FloatFromCurve(FloatCurve=DataInterface() {propertyValues {Curve {keys [<at 0 5 RCIM_Cubic>,<at 1 0>]}}}))
    )
    ScaleA=true
    ScaleRGB=true
   )
   &ScaleRibbonWidth(Ribbon Width Scale=&FloatFromCurve(
     FloatCurve=DataInterface() {propertyValues {Curve {keys [
         {arriveTangent=0.1922, interpMode=RCIM_Cubic, leaveTangent=0.1922, tangentMode=RCTM_User, time=0, value=1}
         {arriveTangent=-0.2774, interpMode=RCIM_Cubic, leaveTangent=-1.7991, tangentMode=RCTM_Break, time=0.5, value=0.625}
         {arriveTangent=-0.5166, interpMode=RCIM_Cubic, leaveTangent=-0.5166, tangentMode=RCTM_User, time=1, value=0}
        ]}}}
     Scale Curve=1.5
    ))
  ]
  rendererClasses [&NiagaraRibbonRendererProperties]
  renderers [{rendererClass=&NiagaraRibbonRendererProperties}]
 }
 BlackTrail002 {
  emitterUpdateScript [
   &EmitterState(Life Cycle Mode=::NewEnumerator1, Loop Behavior=::NewEnumerator1, Loop Delay=0.1, Loop Duration=0.7, UseLoopDelay=true)
   &FN_SpawnAnimTrail(Number of Anim Trails:NiagaraInt32=&SelectIntFromArray(
     Direct Array Index=&MakeIntFromFloat(Float=$User.VariantSwitch)
     Int Selection Array=DataInterface() {propertyValues {IntData [0,0,1,1,1,1]}}
    ))
  ]
  particleSpawnScript [
   &FN_InitializeAnimTrailBoneIndecies()
   &FN_InitializeAnimTrail(Lifetime=0.3, SkeletalMesh=$System.Skeletal Mesh)
  ]
  particleUpdateScript [
   &ParticleState()
   &FN_UpdateAnimTrail(SkeletalMesh=$System.Skeletal Mesh)
   &ScaleColor(
    Scale Alpha=&SelectFloatFromArray(
     Direct Array Index=&MakeIntFromFloat(Float=$User.VariantSwitch)
     Float Selection Array=DataInterface() {propertyValues {FloatData [0,0,1,1,1,1]}}
    )
    Scale RGB=&MakeVectorFromLinearColorRGB(Color=&SelectLinearColorFromArray(
      Array Sampling Mode=::NewEnumerator1
      Direct Array Index=&MakeIntFromFloat(Float=$User.VariantSwitch)
      Color Selection Array=DataInterface() {propertyValues {ColorData [<rgba 0 0 0 1>, <rgba 0 0 0 1>, <rgba 0.8933 2.1226 5 1>, <rgba 8 2.456 0.08 1>, <rgba 0.3209 0.8458 1 1>]}}
      Random Seed=?
     ))
    ScaleA=true
    ScaleRGB=true
   ) {module=ScaleColor001}
   &ScaleColor(
    Scale Alpha=&Multiply_Float(
     A=&FloatFromCurve(FloatCurve=DataInterface() {propertyValues {Curve {keys [
          {interpMode=RCIM_Cubic, leaveTangent=-2.3646, tangentMode=RCTM_Break, time=0, value=1}
          {arriveTangent=-0.4446, interpMode=RCIM_Cubic, tangentMode=RCTM_Break, time=1, value=0}
         ]}}})
     B=&MakeCustomFloatFromBool(Bool=$User.bIsFrontEndPreview, True Float=1, False Float=2)
    )
    Scale RGB=&Multiply_Vector(
     A=<1 1 1>
     B=&VectorFromFloat(Value=&FloatFromCurve(FloatCurve=DataInterface() {propertyValues {Curve {keys [<at 0 5 RCIM_Cubic>,<at 1 0>]}}}))
    )
    ScaleA=true
    ScaleRGB=true
   )
   &ScaleRibbonWidth(Ribbon Width Scale=&FloatFromCurve(
     FloatCurve=DataInterface() {propertyValues {Curve {keys [
         {arriveTangent=0.1922, interpMode=RCIM_Cubic, leaveTangent=0.1922, tangentMode=RCTM_User, time=0, value=1}
         {arriveTangent=-0.2774, interpMode=RCIM_Cubic, leaveTangent=-1.7991, tangentMode=RCTM_Break, time=0.5, value=0.625}
         {arriveTangent=-0.5166, interpMode=RCIM_Cubic, leaveTangent=-0.5166, tangentMode=RCTM_User, time=1, value=0}
        ]}}}
     Scale Curve=1.5
    ))
  ]
  rendererClasses [&NiagaraRibbonRendererProperties]
  renderers [{rendererClass=&NiagaraRibbonRendererProperties}]
 }
 BlackTrail_Galaxy {
  emitterUpdateScript [
   &EmitterState(Life Cycle Mode=::NewEnumerator1, Loop Behavior=::NewEnumerator1, Loop Delay=0.1, Loop Duration=0.7, UseLoopDelay=true)
   &FN_SpawnAnimTrail(Number of Anim Trails:NiagaraInt32=&SelectIntFromArray(
     Direct Array Index=&MakeIntFromFloat(Float=$User.VariantSwitch)
     Int Selection Array=DataInterface() {propertyValues {IntData [0,0,1,0,0,0]}}
    ))
  ]
  particleSpawnScript [
   &FN_InitializeAnimTrailBoneIndecies()
   &FN_InitializeAnimTrail(AnimTrailWidth=2, Color=<rgba 5 5 5 1>, Lifetime=0.4, SkeletalMesh=$System.Skeletal Mesh)
  ]
  particleUpdateScript [
   &DynamicMaterialParameters(
    Index 1 Param 1=&FloatFromCurve(FloatCurve=DataInterface() {propertyValues {Curve {keys [{tangentMode=RCTM_User, time=0, value=1}, {tangentMode=RCTM_User, time=1, value=1}]}}})
    Index 1 Param 3=&FloatFromCurve(FloatCurve=DataInterface() {propertyValues {Curve {keys [<at 0 -1 RCIM_Cubic>, <at 0.55 0.6116>, {arriveTangent=-1.3527, tangentMode=RCTM_User, time=1, value=1}]}}})
    Module.Index0Param0Write=true
    Module.Index0Param1Write=true
    Module.Index0Param2Write=false
    Module.Index0Param3Write=false
    Module.Index1Param0Write=true
    Module.Index1Param1Write=false
    Module.Index1Param2Write=true
    Module.Index1Param3Write=false
    Write Parameter Index 1=true
   )
   &ParticleState()
   &FN_UpdateAnimTrail(SkeletalMesh=$System.Skeletal Mesh)
   &ScaleColor(
    Scale Alpha=0.5
    Scale RGB=&MakeVectorFromLinearColorRGB(Color=&SelectLinearColorFromArray(
      Array Sampling Mode=::NewEnumerator1
      Direct Array Index=&MakeIntFromFloat(Float=$User.VariantSwitch)
      Color Selection Array=DataInterface() {propertyValues {ColorData [<rgba 0.3699 0.0488 1 2>, <rgba 0.0724 0.257 1 2>, <rgba 0.75 2.1165 8 1>, <rgba 1 0.3487 0.0695 2>, <rgba 0 0 0 0>, <rgba 0 0 0 0>]}}
      Random Seed=?
     ))
    ScaleA=true
    ScaleRGB=true
   ) {module=ScaleColor001}
   &ScaleColor(
    Scale Alpha=&FloatFromCurve(
     FloatCurve=DataInterface() {propertyValues {Curve {keys [
         {arriveTangent=-1.0897, interpMode=RCIM_Cubic, leaveTangent=-1.0897, tangentMode=RCTM_User, time=0, value=1}
         {arriveTangent=-1.9706, interpMode=RCIM_Cubic, leaveTangent=-1.9706, tangentMode=RCTM_User, time=1, value=0}
        ]}}}
     Scale Curve=&FloatFromCurve(
      FloatCurve=DataInterface() {propertyValues {Curve {keys [
          <at 0 1>
          <at 0.2 1>
          {arriveTangent=-5.793, interpMode=RCIM_Cubic, leaveTangent=-5.793, tangentMode=RCTM_User, time=0.3, value=1}
          <at 0.8 0>
         ]}}}
      CurveIndex=$Emitter.Age
      Scale Curve=1.3
     )
    )
    Scale RGB=&Multiply_VectorByFloat(
     Vector=<1 1 1>
     Float=&FloatFromCurve(FloatCurve=DataInterface() {propertyValues {Curve {keys [<at 0 1>,<at 0.35 0.1>]}}})
    )
    ScaleA=true
    ScaleRGB=true
   )
   &ScaleRibbonWidth(Ribbon Width Scale=&FloatFromCurve(FloatCurve=DataInterface() {propertyValues {Curve {
        defaultValue=1
        keys [
         {interpMode=RCIM_Cubic, tangentMode=RCTM_User, time=0, value=1}
         {arriveTangent=-2.8868, interpMode=RCIM_Cubic, leaveTangent=-2.8868, tangentMode=RCTM_User, time=1, value=0}
        ]
       }}}))
  ]
  rendererClasses [&NiagaraRibbonRendererProperties]
  renderers [{rendererClass=&NiagaraRibbonRendererProperties}]
 }
 BlackTrail_Holifoil {
  emitterUpdateScript [
   &EmitterState(Life Cycle Mode=::NewEnumerator1, Loop Behavior=::NewEnumerator1, Loop Delay=0.1, Loop Duration=0.7, UseLoopDelay=true)
   &FN_SpawnAnimTrail(Number of Anim Trails:NiagaraInt32=&SelectIntFromArray(
     Direct Array Index=&MakeIntFromFloat(Float=$User.VariantSwitch)
     Int Selection Array=DataInterface() {propertyValues {IntData [0,0,0,0,1,0]}}
    ))
  ]
  particleSpawnScript [
   &FN_InitializeAnimTrailBoneIndecies()
   &FN_InitializeAnimTrail(AnimTrailWidth=2, Color=<rgba 4 4 4 1>, Lifetime=0.4, SkeletalMesh=$System.Skeletal Mesh)
  ]
  particleUpdateScript [
   &DynamicMaterialParameters(
    Index 1 Param 1=&FloatFromCurve(FloatCurve=DataInterface() {propertyValues {Curve {keys [{tangentMode=RCTM_User, time=0, value=1}, {tangentMode=RCTM_User, time=1, value=1}]}}})
    Index 1 Param 3=&FloatFromCurve(FloatCurve=DataInterface() {propertyValues {Curve {keys [<at 0 -1 RCIM_Cubic>, <at 0.55 0.6116>, {arriveTangent=-1.3527, tangentMode=RCTM_User, time=1, value=1}]}}})
    Module.Index0Param0Write=true
    Module.Index0Param1Write=true
    Module.Index0Param2Write=false
    Module.Index0Param3Write=false
    Module.Index1Param0Write=true
    Module.Index1Param1Write=false
    Module.Index1Param2Write=true
    Module.Index1Param3Write=false
    Write Parameter Index 1=true
   )
   &ParticleState()
   &FN_UpdateAnimTrail(SkeletalMesh=$System.Skeletal Mesh)
   &ScaleColor(
    Scale RGB=&MakeVectorFromLinearColorRGB(Color=&SelectLinearColorFromArray(
      Array Sampling Mode=::NewEnumerator1
      Direct Array Index=&MakeIntFromFloat(Float=$User.VariantSwitch)
      Color Selection Array=DataInterface() {propertyValues {ColorData [<rgba 1 1 1 2>, <rgba 1 1 1 2>, <rgba 2 2 2 1>, <rgba 1 1 1 2>, <rgba 0.6683 1.3995 1.5 1>, <rgba 1 1 1 1>]}}
      Random Seed=?
     ))
    ScaleA=true
    ScaleRGB=true
   ) {module=ScaleColor001}
   &ScaleColor(
    Scale Alpha=&FloatFromCurve(
     FloatCurve=DataInterface() {propertyValues {Curve {keys [
         {arriveTangent=-1.0897, interpMode=RCIM_Cubic, leaveTangent=-1.0897, tangentMode=RCTM_User, time=0, value=1}
         {arriveTangent=-1.9706, interpMode=RCIM_Cubic, leaveTangent=-1.9706, tangentMode=RCTM_User, time=1, value=0}
        ]}}}
     Scale Curve=&FloatFromCurve(
      FloatCurve=DataInterface() {propertyValues {Curve {keys [
          <at 0 1>
          <at 0.2 1>
          {arriveTangent=-5.793, interpMode=RCIM_Cubic, leaveTangent=-5.793, tangentMode=RCTM_User, time=0.3, value=1}
          <at 0.8 0>
         ]}}}
      CurveIndex=$Emitter.Age
     )
    )
    Scale RGB=&Multiply_VectorByFloat(
     Vector=<1 1 1>
     Float=&FloatFromCurve(FloatCurve=DataInterface() {propertyValues {Curve {keys [<at 0 1>,<at 0.35 0.1>]}}})
    )
    ScaleA=true
    ScaleRGB=true
   )
   &ScaleRibbonWidth(Ribbon Width Scale=&FloatFromCurve(FloatCurve=DataInterface() {propertyValues {Curve {
        defaultValue=1
        keys [
         {interpMode=RCIM_Cubic, tangentMode=RCTM_User, time=0, value=1}
         {arriveTangent=-2.8868, interpMode=RCIM_Cubic, leaveTangent=-2.8868, tangentMode=RCTM_User, time=1, value=0}
        ]
       }}}))
  ]
  rendererClasses [&NiagaraRibbonRendererProperties]
  renderers [{rendererClass=&NiagaraRibbonRendererProperties}]
 }
 Sparks {
  emitterUpdateScript [
   &EmitterState(Life Cycle Mode=::NewEnumerator1, Loop Behavior=::NewEnumerator1, Loop Duration=0.5, UseLoopDelay=false)
   &SpawnPerUnit(Module.Use Max Movement Threshold=true, Module.Use Movement Tolerance=false, Use Spawn Probability=false)
  ]
  particleSpawnScript [
   &InitializeParticle(
    Color=&SelectLinearColorFromArray(
     Array Sampling Mode=::NewEnumerator1
     Direct Array Index=&MakeIntFromFloat(Float=$User.VariantSwitch)
     Color Selection Array=DataInterface() {propertyValues {ColorData [<rgba 2.9213 1.0815 10 1>, <rgba 1.0949 4.8144 10 1>, <rgba 0.8294 5.7262 10 1>, <rgba 10 3.7665 1.0949 1>, <rgba 1.1858 6.3538 10 1>]}}
     Random Seed=?
    )
    Lifetime Max=0.6
    Lifetime Min=0.4
    Lifetime Mode=::NewEnumerator1
    Sprite Size Max=<1 13>
    Sprite Size Min=<2 10>
    Sprite Size Mode=::NewEnumerator2
    UsePositionOffset=false
    Write Material Random=true
   )
   &SkeletalMeshLocation(
    Bone / Socket Index=1
    MaskSpawnGroup=false
    Mesh Sampling Type=::NewEnumerator5
    Skeletal Mesh=$System.Skeletal Mesh
    Socket Sampling Mode=::NewEnumerator5
   )
   &AddVelocity()
   &BoxLocation(Box Size=<20 1 1>, MaskSpawnGroup=false, Module.Override Seed=false, Rotate Box=false, Surface Only=false)
   &ApplyInitialForces()
   &InheritVelocity(Speed Limit=true, UseSpeedThreshold=false)
  ]
  particleUpdateScript [
   &ParticleState()
   /SetVariables(Particles.Glow03_Color=&MakeLinearColorFromVectorAndFloat(
     "Vector (RGB)"=&MakeVectorFromLinearColorRGB(Color=&MultiplyLinearColor(
       Linear Color=&SelectLinearColorFromArray(
        Array Sampling Mode=::NewEnumerator1
        Direct Array Index=&MakeIntFromFloat(Float=$User.VariantSwitch)
        Color Selection Array=DataInterface() {propertyValues {ColorData [<rgba 0.5111 0.0795 1 0.2>, <rgba 0 0.2055 1 0.2>, <rgba 0.2126 0.0695 1 0.5>, <rgba 1 0.3635 0.037 0.2>, <rgba 0.409 0.1252 1 0.2>]}}
        Random Seed=?
       )
       Scale Factor=&MakeVector4_FromFloat(Float=1)
      ))
     "Float (Alpha)"=&Multiply_Float(A=0.2, B=1)
    )) {isSetParameters=true}
   /SetVariables(Particles.Glow03_Size=&Multiply_Vector2D_ByFloat(
     Float=&FloatFromCurve(
      FloatCurve=DataInterface() {propertyValues {Curve {keys [
          <at 0 0 RCIM_Cubic>
          <at 0.2 1.3 RCIM_Cubic>
          {arriveTangent=-0.1423, interpMode=RCIM_Cubic, leaveTangent=-0.1423, tangentMode=RCTM_User, time=1, value=0}
         ]}}}
      Scale Curve=15
     )
     Vector2D=<0.7 1.5>
    )) {isSetParameters=true}
   &ScaleColor(
    Scale Alpha=&FloatFromCurve(FloatCurve=DataInterface() {propertyValues {Curve {keys [<at 0.3 1>,<at 0.4 1>,<at 1 1>]}}})
    ScaleA=true
    ScaleRGB=true
   )
   &ScaleSpriteSize(Scale Factor=&Vector2DFromFloat(Value=&FloatFromCurve(
      FloatCurve=DataInterface() {propertyValues {Curve {keys [
          <at 0 0 RCIM_Cubic>
          {interpMode=RCIM_Cubic, leaveTangent=-3.0296, tangentMode=RCTM_Break, time=0.2, value=1.2}
          {arriveTangent=-1.8107, interpMode=RCIM_Cubic, leaveTangent=-1.8107, tangentMode=RCTM_User, time=1, value=0}
         ]}}}
      Scale Curve=0.4
     ))) {module=ScaleSpriteSize001}
   &GravityForce(Gravity=<0 0 -100>)
   &CurlNoiseForce(
    Noise Frequency=300
    Noise Strength=&Multiply_Float(A=1000, B=1)
    Pan Noise=true
    Pan Noise Field=<0 0 1>
    Random Seed=10
   )
   &Drag(Drag=3, Use Linear Drag=true, Use Rotational Drag=false)
   &SolveForcesAndVelocity(Clamp Velocity=false, Limit Acceleration=false)
  ]
  rendererClasses [&NiagaraSpriteRendererProperties]
  renderers [
   {rendererClass=&NiagaraSpriteRendererProperties}
   {rendererClass=&NiagaraSpriteRendererProperties, rendererIndex=1}
  ]
 }
 TrialEnergy001 {
  emitterUpdateScript [
   &EmitterState(Inactive Response=::NewEnumerator1, Life Cycle Mode=::NewEnumerator1, Loop Behavior=::NewEnumerator1, Loop Duration=1, UseLoopDelay=false)
   &FN_SpawnAnimTrail()
  ]
  particleSpawnScript [
   &FN_InitializeAnimTrailBoneIndecies()
   &FN_InitializeAnimTrail(
    AnimTrailWidth=2
    Color=&MakeCustomLinearColorFromBool(Bool=$User.bIsFrontEndPreview, True Color=<rgba 0.5 0.5 0.5 1>, False Color=<rgba 1 1 1 1>)
    Lifetime=0.7
    SkeletalMesh=$System.Skeletal Mesh
   )
   &DynamicMaterialParameters(
    Index 0 Param 1=1
    Index 0 Param 2=1
    Index 0 Param 3=0.5
    Module.Index0Float0Write=true
    Module.Index0Float1Write=true
    Module.Index0Float2Write=true
    Module.Index0Float3Write=true
   )
  ]
  particleUpdateScript [
   &ParticleState()
   &FN_UpdateAnimTrail(SkeletalMesh=$System.Skeletal Mesh)
   &ScaleColor(
    Scale Alpha=&MakeCustomFloatFromBool(Bool=$User.bIsFrontEndPreview, True Float=1, False Float=2)
    Scale RGB=/Niagara/DynamicInputs/Math/LerpMultipleVectors(
     Alpha:NiagaraFloat=&Divide_Float(A=$User.VariantSwitch, B=4)
     V 1:Vector3f=&MakeVectorFromLinearColorRGB(Color=&ColorFromCurve(ColorCurve=DataInterface() {propertyValues {
         AlphaCurve {keys [<at 0 1 RCIM_Cubic>,<at 1 0 RCIM_Cubic>]}
         BlueCurve {keys [<at 0.1598 1 RCIM_Cubic>,<at 0.7697 1 RCIM_Cubic>]}
         GreenCurve {keys [<at 0.1598 0.0488 RCIM_Cubic>,<at 0.7697 0.0488 RCIM_Cubic>]}
         RedCurve {keys [<at 0.1598 0.3699 RCIM_Cubic>,<at 0.7697 0.3699 RCIM_Cubic>]}
        }}))
     V 2:Vector3f=&MakeVectorFromLinearColorRGB(Color=&ColorFromCurve(ColorCurve=DataInterface() {propertyValues {
         AlphaCurve {keys [<at 0 1 RCIM_Cubic>,<at 1 0 RCIM_Cubic>]}
         BlueCurve {keys [<at 0.4353 1 RCIM_Cubic>,<at 0.678 1 RCIM_Cubic>]}
         GreenCurve {keys [<at 0.4353 0.318 RCIM_Cubic>,<at 0.678 0.1281 RCIM_Cubic>]}
         RedCurve {keys [<at 0.4353 0 RCIM_Cubic>,<at 0.678 0.1 RCIM_Cubic>]}
        }}))
     V 3:Vector3f=&MakeVectorFromLinearColorRGB(Color=&ColorFromCurve(ColorCurve=DataInterface() {propertyValues {
         AlphaCurve {keys [<at 0 1 RCIM_Cubic>,<at 1 0 RCIM_Cubic>]}
         BlueCurve {keys [<at 0.2693 0.2448 RCIM_Cubic>,<at 0.7845 1 RCIM_Cubic>]}
         GreenCurve {keys [<at 0.2693 0.1346 RCIM_Cubic>,<at 0.7845 0.1264 RCIM_Cubic>]}
         RedCurve {keys [<at 0.2693 0.0244 RCIM_Cubic>,<at 0.7845 0.331 RCIM_Cubic>]}
        }}))
     V 4:Vector3f=&MakeVectorFromLinearColorRGB(Color=&ColorFromCurve(ColorCurve=DataInterface() {propertyValues {
         AlphaCurve {keys [<at 0 1 RCIM_Cubic>,<at 1 0 RCIM_Cubic>]}
         BlueCurve {keys [<at 0.1838 0.0142 RCIM_Cubic>,<at 0.6454 0.0371 RCIM_Cubic>]}
         GreenCurve {keys [<at 0.1838 0.1584 RCIM_Cubic>,<at 0.6454 0.2572 RCIM_Cubic>]}
         RedCurve {keys [<at 0.1838 0.3073 RCIM_Cubic>,<at 0.6454 0.5729 RCIM_Cubic>]}
        }}))
     V 5:Vector3f=&MakeVectorFromLinearColorRGB(Color=&ColorFromCurve(ColorCurve=DataInterface() {propertyValues {
         AlphaCurve {keys [<at 0 1 RCIM_Cubic>,<at 1 0 RCIM_Cubic>]}
         BlueCurve {keys [<at 0.2163 0.4427 RCIM_Cubic>,<at 0.5146 1 RCIM_Cubic>]}
         GreenCurve {keys [<at 0.2163 0.2889 RCIM_Cubic>,<at 0.5146 0.125 RCIM_Cubic>]}
         RedCurve {keys [<at 0.2163 0.0807 RCIM_Cubic>,<at 0.5146 0.6312 RCIM_Cubic>]}
        }}))
     Vector Count:ENiagara_MultipleLerpCount=::NewEnumerator2
    )
    ScaleA=true
    ScaleRGB=true
   ) {module=ScaleColor001}
   &ScaleColor(
    Scale Alpha=&FloatFromCurve(
     FloatCurve=DataInterface() {propertyValues {Curve {keys [
         {interpMode=RCIM_Cubic, leaveTangent=-2.3646, tangentMode=RCTM_Break, time=0, value=1}
         {arriveTangent=-0.8437, interpMode=RCIM_Cubic, leaveTangent=-0.8437, tangentMode=RCTM_User, time=0.2974, value=0.2064}
         {interpMode=RCIM_Cubic, tangentMode=RCTM_Break, time=1, value=0}
        ]}}}
     Scale Curve=4
    )
    Scale RGB=&MakeVectorFromLinearColorRGB(Color=&ColorFromCurve(ColorCurve=DataInterface() {propertyValues {
        AlphaCurve {keys [<at 0 1 RCIM_Cubic>]}
        BlueCurve {keys [<at 0 6 RCIM_Cubic>,<at 0.205 1 RCIM_Cubic>]}
        GreenCurve {keys [<at 0 4.6324 RCIM_Cubic>,<at 0.205 1 RCIM_Cubic>]}
        RedCurve {keys [<at 0 3.0919 RCIM_Cubic>,<at 0.205 1 RCIM_Cubic>]}
       }}))
    ScaleA=true
    ScaleRGB=true
   )
   &ScaleRibbonWidth(Ribbon Width Scale=&FloatFromCurve(
     FloatCurve=DataInterface() {propertyValues {Curve {keys [
         {interpMode=RCIM_Cubic, tangentMode=RCTM_User, time=0, value=1}
         {arriveTangent=-2.0454, interpMode=RCIM_Cubic, leaveTangent=-2.0454, tangentMode=RCTM_User, time=1, value=0}
        ]}}}
     Scale Curve=1.5
    )) {enabled=false}
  ]
  rendererClasses [&NiagaraRibbonRendererProperties]
  renderers [{rendererClass=&NiagaraRibbonRendererProperties}]
 }
 TrialEnergy004 {
  emitterUpdateScript [
   &EmitterState(Inactive Response=::NewEnumerator1, Life Cycle Mode=::NewEnumerator1, Loop Behavior=::NewEnumerator1, Loop Duration=0.5, UseLoopDelay=true)
   &FN_SpawnAnimTrail()
  ]
  particleSpawnScript [
   &FN_InitializeAnimTrailBoneIndecies()
   &FN_InitializeAnimTrail(AnimTrailWidth=4.5, Lifetime=0.7, SkeletalMesh=$System.Skeletal Mesh)
  ]
  particleUpdateScript [
   &ParticleState()
   &FN_UpdateAnimTrail(SkeletalMesh=$System.Skeletal Mesh)
   &ScaleColor(
    Scale Alpha=&FloatFromCurve(
     FloatCurve=DataInterface() {propertyValues {Curve {keys [
         {interpMode=RCIM_Cubic, leaveTangent=-2.3646, tangentMode=RCTM_Break, time=0, value=1}
         {arriveTangent=-0.7781, interpMode=RCIM_Cubic, tangentMode=RCTM_Break, time=1, value=0}
        ]}}}
     Scale Curve=2
    )
    Scale RGB=&LerpMultipleVectors(
     Alpha:NiagaraFloat=&Divide_Float(A=$User.VariantSwitch, B=4)
     V 1:Vector3f=&MakeVectorFromLinearColorRGB(Color=&ColorFromCurve(ColorCurve=DataInterface() {propertyValues {
         AlphaCurve {keys [<at 0 1 RCIM_Cubic>,<at 1 0 RCIM_Cubic>]}
         BlueCurve {keys [<at 0.1598 1 RCIM_Cubic>,<at 0.7697 2 RCIM_Cubic>]}
         GreenCurve {keys [<at 0.1598 0.0521 RCIM_Cubic>,<at 0.7697 0.4254 RCIM_Cubic>]}
         RedCurve {keys [<at 0.1598 0.526 RCIM_Cubic>,<at 0.7697 0.819 RCIM_Cubic>]}
        }}))
     V 2:Vector3f=&MakeVectorFromLinearColorRGB(Color=&ColorFromCurve(ColorCurve=DataInterface() {propertyValues {
         AlphaCurve {keys [<at 0 1 RCIM_Cubic>,<at 1 0 RCIM_Cubic>]}
         BlueCurve {keys [<at 0.1598 1 RCIM_Cubic>,<at 0.7697 2 RCIM_Cubic>]}
         GreenCurve {keys [<at 0.1598 0.1494 RCIM_Cubic>,<at 0.7697 0.5501 RCIM_Cubic>]}
         RedCurve {keys [<at 0.1598 0.0393 RCIM_Cubic>,<at 0.7697 0.125 RCIM_Cubic>]}
        }}))
     V 3:Vector3f=&MakeVectorFromLinearColorRGB(Color=&ColorFromCurve(ColorCurve=DataInterface() {propertyValues {
         AlphaCurve {keys [<at 0 1 RCIM_Cubic>,<at 1 0 RCIM_Cubic>]}
         BlueCurve {keys [<at 0.1285 0.5521 RCIM_Cubic>, <at 0.3222 0.3228>, <at 0.4888 0.3228>, {interpMode=RCIM_Cubic, leaveTangent=11.2128, time=0.769, value=2}]}
         GreenCurve {keys [
           <at 0.1285 0.2788 RCIM_Cubic>
           <at 0.3222 0.0162>
           <at 0.4888 0.0162>
           {interpMode=RCIM_Cubic, leaveTangent=0.9605, time=0.769, value=0.1668}
          ]}
         RedCurve {keys [
           <at 0.1285 0.0575 RCIM_Cubic>
           <at 0.3222 0.2435>
           <at 0.4888 0.2435>
           {interpMode=RCIM_Cubic, leaveTangent=4.6471, time=0.769, value=0.8728}
          ]}
        }}))
     V 4:Vector3f=&MakeVectorFromLinearColorRGB(Color=&ColorFromCurve(ColorCurve=DataInterface() {propertyValues {
         AlphaCurve {keys [<at 0 1 RCIM_Cubic>,<at 1 0 RCIM_Cubic>]}
         BlueCurve {keys [<at 0.1598 0.0698 RCIM_Cubic>,<at 0.7697 0.0136 RCIM_Cubic>]}
         GreenCurve {keys [<at 0.1598 0.4859 RCIM_Cubic>,<at 0.7697 0.1001 RCIM_Cubic>]}
         RedCurve {keys [<at 0.1598 1 RCIM_Cubic>,<at 0.7697 0.3021 RCIM_Cubic>]}
        }}))
     V 5:Vector3f=&MakeVectorFromLinearColorRGB(Color=&ColorFromCurve(ColorCurve=DataInterface() {propertyValues {
         AlphaCurve {keys [<at 0 1 RCIM_Cubic>,<at 1 0 RCIM_Cubic>]}
         BlueCurve {keys [<at 0.2254 1 RCIM_Cubic>,<at 0.5702 2 RCIM_Cubic>]}
         GreenCurve {keys [<at 0.2254 0.4905 RCIM_Cubic>,<at 0.5702 0.3412 RCIM_Cubic>]}
         RedCurve {keys [<at 0.2254 0.1398 RCIM_Cubic>,<at 0.5702 1.7059 RCIM_Cubic>]}
        }}))
     Vector Count:ENiagara_MultipleLerpCount=::NewEnumerator2
    )
    ScaleA=true
    ScaleRGB=true
   )
   &ScaleRibbonWidth(Ribbon Width Scale=&FloatFromCurve(
     FloatCurve=DataInterface() {propertyValues {Curve {keys [
         {interpMode=RCIM_Cubic, tangentMode=RCTM_User, time=0, value=1}
         {arriveTangent=-2.0454, interpMode=RCIM_Cubic, leaveTangent=-2.0454, tangentMode=RCTM_User, time=1, value=0}
        ]}}}
     Scale Curve=1.5
    )) {enabled=false}
   &DynamicMaterialParameters(
    Index 0 Param 3=-0.8
    Index 0 Param 4=&FloatFromCurve(FloatCurve=DataInterface() {propertyValues {Curve {keys [
         {interpMode=RCIM_Cubic, tangentMode=RCTM_User, time=0, value=1}
         {arriveTangent=-0.6219, interpMode=RCIM_Cubic, leaveTangent=-0.6219, tangentMode=RCTM_User, time=1, value=0}
        ]}}})
    Module.Index0Float0Write=true
    Module.Index0Float1Write=true
    Module.Index0Float2Write=true
    Module.Index0Float3Write=true
   )
  ]
  rendererClasses [&NiagaraRibbonRendererProperties]
  renderers [{rendererClass=&NiagaraRibbonRendererProperties}]
 }
}
summary {
 systemName=NS_Pickaxe_TidalNinja_Swing
 emitters [
  {emitterName=TrialEnergy001, rendererClasses [&NiagaraRibbonRendererProperties]}
  {emitterName=BlackTrail, rendererClasses [&NiagaraRibbonRendererProperties]}
  {emitterName=BlackTrail001, rendererClasses [&NiagaraRibbonRendererProperties]}
  {emitterName=TrialEnergy004, rendererClasses [&NiagaraRibbonRendererProperties]}
  {emitterName=Sparks, rendererClasses [&NiagaraSpriteRendererProperties]}
  {emitterName=BlackTrail002, rendererClasses [&NiagaraRibbonRendererProperties]}
  {emitterName=BlackTrail_Galaxy, rendererClasses [&NiagaraRibbonRendererProperties]}
  {emitterName=BlackTrail_Holifoil, rendererClasses [&NiagaraRibbonRendererProperties]}
 ]
 userVariables [
  User.bIsFrontEnd {type=NiagaraBool}
  User.bIsFrontEndPreview {type=NiagaraBool}
  User.RibbonIDOffset {type=NiagaraInt32}
  User.VariantSwitch {type=NiagaraFloat, defaultValue=0, description=""}
 ]
}
